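# One-shot stereo MetaSound for the instigator's melee body hitmarker, part of the gunplay update hit notifications.
# It layers a stereo wave with a mono wave whose pitch rises with accumulated damage, remapped through a shared wavetable bank curve.
# The mix is scaled by a hitmarker volume modulator and finishes when both players end.
# Bus sends feed sidechain buses for weapons, explosions, ambience and other categories.
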MetaSoundSource /Game/Athena/Sounds/HitNotifications/Gunplay_Update/Melee/MSS_Hitmarker_Melee_Instigator_Body
inputs [
 HitMarker.Azimuth {type=Float}
 HitMarker.DamageDealt {type=Float}
 HitMarker.Distance {type=Float}
 HitMarker.InstigatorAccumulatedDamage {type=Float}
 HitMarker.ReceiverPercentHealth {type=Float}
 HitMarker.ReceiverPercentShield {type=Float}
 RecentCollection.NumRecentPlays {type=Int32}
 RecentCollection.TimeSinceLastPlayed {type=Float}
 UE.Source.OnPlay {type=Trigger}
]
interfaces [Fort.ActiveSoundRecentCollection:1.0, Fort.HitMarker:1.0, UE.OutputFormat.Stereo:1.0, UE.Source.OneShot:1.0, UE.Source:1.1]
outputs [
 "UE.OutputFormat.Stereo.Audio:0" {type=Audio}
 "UE.OutputFormat.Stereo.Audio:1" {type=Audio}
 UE.Source.OneShot.OnFinished {type=Trigger}
]
pages [Default {
  input_defaults {
   HitMarker.Azimuth=0
   HitMarker.DamageDealt=0
   HitMarker.Distance=0
   HitMarker.InstigatorAccumulatedDamage=0
   HitMarker.ReceiverPercentHealth=1
   HitMarker.ReceiverPercentShield=1
   RecentCollection.NumRecentPlays=0
   RecentCollection.TimeSinceLastPlayed=0
   UE.Source.OnPlay=null
  }
  nodes [
   n1 = MapRange.MapRange:Float(In=^HitMarker.InstigatorAccumulatedDamage) {also {defaults {In Range B=300}}}
   WavePlayer_0 = UE.Wave Player:Mono(
    Play=^UE.Source.OnPlay
    Pitch Shift=MapRange.MapRange:Float(In=UE.WaveTableBankEvaluate(Input=@n1.Out Value) {also {defaults {Index=1, WaveTableBank=/Game/Athena/Sounds/HitNotifications/Gunplay_Update/WB_Hitmarker_GunplayUpdate}}}.Output) {also {defaults {Out Range A=2, Out Range B=10}}}.Out Value
    Wave Asset=VariableAccessor.WaveAsset(Variable=InitVariable.WaveAsset().Variable).Value
   ) {also {version [1,1]}}
   WavePlayer_1 = UE.Wave Player:Stereo(Play=^UE.Source.OnPlay, Wave Asset=VariableAccessor.WaveAsset(Variable=InitVariable.WaveAsset().Variable).Value) {also {version [1,1]}}
   AudioMixerAudioMixerStereo2_0 = "AudioMixer.Audio Mixer (Stereo, 2)"(In 1 L=@WavePlayer_0.Out Mono, In 1 R=@WavePlayer_0.Out Mono, In 0 L=@WavePlayer_1.Out Left, In 0 R=@WavePlayer_1.Out Right) {also {defaults {Gain 0=0.6, Gain 1=1.2}}}
   n2 = Modulation.GetModulatorValue() {also {defaults {Modulator=/Game/Athena/Sounds/HitNotifications/Gunplay_Update/Modulation/Patch_Hitmarkers_Instigator_Standard_Volume}}}
   Multiply_0 = UE.Multiply:Audio by Float(PrimaryOperand=@AudioMixerAudioMixerStereo2_0.Out R, AdditionalOperands=@n2.Out)
   Multiply_1 = UE.Multiply:Audio by Float(PrimaryOperand=@AudioMixerAudioMixerStereo2_0.Out L, AdditionalOperands=@n2.Out)
   TriggerAccumulatorTriggerAccumulate2_0 = "TriggerAccumulator.Trigger Accumulate (2)"(In 1=@WavePlayer_0.On Finished, In 0=@WavePlayer_1.On Finished)
  ]
  links [
   {dst="@UE.OutputFormat.Stereo.Audio:1", src=Multiply_0.Out}
   {dst="@UE.OutputFormat.Stereo.Audio:0", src=Multiply_1.Out}
   {dst="@UE.Source.OneShot.OnFinished", src=TriggerAccumulatorTriggerAccumulate2_0.Out}
  ]
 }]
properties {
 assetImportData=&self:AssetImportData
 audioPropertiesSheet=/Game/Athena/Sounds/HitNotifications/Gunplay_Update/APS_Hitmarker_Instigator_Standard
 numChannels=2
 outputFormat=Stereo
 sampleRate=48000
 soundClassObject=&Fort_Hitmarkers
 soundSubmixObject=&Fortnite_SFX
 volume=2
 busSends [
  {audioBus=&AB_Sidechain_FortWeapon_Vol, sendLevel=0.35}
  {audioBus=&AB_Sidechain_FortExplosion_Vol, sendLevel=0.35}
  {audioBus=&AB_Sidechain_FortAmbient_Vol, sendLevel=0.3}
  {audioBus=&AB_Sidechain_FortLoot_Vol, sendLevel=0.3}
  {audioBus=&AB_Sidechain_FortVehicles_Vol, sendLevel=0.3}
  {audioBus=&AB_Sidechain_FortCreatures_Vol, sendLevel=0.3}
  {audioBus=&AB_Sidechain_FortEmote_Vol, sendLevel=0.3}
 ]
 concurrencySet [/Game/Athena/Sounds/HitNotifications/Gunplay_Update/CON_Hitmarker_GunplayUpdate_Instigator_Standard]
 modulationSettings {
  volumeRouting=Union
  volumeModulationDestination {bEnableModulation=true, modulators [&Patch_FortHitNotify_Vol]}
 }
}
