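# One-shot stereo MetaSound for the hitmarker sound the instigator hears on a melee elimination.
# A wave player plays the Sound input on play and signals finished when the wave ends.
# Both channels are multiplied by the critical instigator hitmarker volume modulator, read inside the graph, alongside source-level hit-notify volume modulation.
# Bus sends feed sidechain buses for weapons, explosions, other game sound categories and standard instigator hitmarkers.
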MetaSoundSource /Game/Athena/Sounds/HitNotifications/Gunplay_Update/Melee/MSS_Hitmarker_Melee_Instigator_Elimination
inputs [
 Sound {type=WaveAsset}
 UE.Source.OnPlay {type=Trigger}
]
interfaces [UE.OutputFormat.Stereo:1.0, UE.Source.OneShot:1.0, UE.Source:1.1]
outputs [
 "UE.OutputFormat.Stereo.Audio:0" {type=Audio}
 "UE.OutputFormat.Stereo.Audio:1" {type=Audio}
 UE.Source.OneShot.OnFinished {type=Trigger}
]
pages [Default {
  input_defaults {Sound=&/Hitmarker_Melee_Instigator_Elim_01, UE.Source.OnPlay=null}
  nodes [
   n1 = Modulation.GetModulatorValue() {also {defaults {Modulator=/Game/Athena/Sounds/HitNotifications/Gunplay_Update/Modulation/Patch_Hitmarkers_Instigator_Critical_Volume}}}
   WavePlayer_0 = UE.Wave Player:Stereo(Wave Asset=^Sound, Play=^UE.Source.OnPlay) {also {version [1,1]}}
   Multiply_0 = UE.Multiply:Audio by Float(AdditionalOperands=@n1.Out, PrimaryOperand=@WavePlayer_0.Out Right)
   Multiply_1 = UE.Multiply:Audio by Float(AdditionalOperands=@n1.Out, PrimaryOperand=@WavePlayer_0.Out Left)
  ]
  links [
   {dst="@UE.OutputFormat.Stereo.Audio:1", src=Multiply_0.Out}
   {dst="@UE.OutputFormat.Stereo.Audio:0", src=Multiply_1.Out}
   {dst="@UE.Source.OneShot.OnFinished", src=WavePlayer_0.On Finished}
  ]
 }]
properties {
 assetImportData=&self:AssetImportData
 audioPropertiesSheet=/Game/Athena/Sounds/HitNotifications/Gunplay_Update/APS_Hitmarker_Instigator_Critical
 numChannels=2
 outputFormat=Stereo
 sampleRate=48000
 soundClassObject=&Fort_Hitmarkers
 soundSubmixObject=&Fortnite_SFX
 volume=1.995
 busSends [
  {audioBus=&AB_Sidechain_FortWeapon_Vol, sendLevel=0.4}
  {audioBus=&AB_Sidechain_FortExplosion_Vol, sendLevel=0.4}
  {audioBus=&AB_Sidechain_FortAmbient_Vol, sendLevel=0.3}
  {audioBus=&AB_Sidechain_FortLoot_Vol, sendLevel=0.3}
  {audioBus=&AB_Sidechain_FortVehicles_Vol, sendLevel=0.3}
  {audioBus=&AB_Sidechain_FortCreatures_Vol, sendLevel=0.3}
  {audioBus=&AB_Sidechain_FortEmote_Vol, sendLevel=0.3}
  {audioBus=/Game/Athena/Sounds/HitNotifications/Gunplay_Update/Modulation/AB_Sidechain_Hitmarkers_Instigator_Standard_Bus, sendLevel=1.5}
 ]
 concurrencySet [/Game/Athena/Sounds/HitNotifications/Gunplay_Update/CON_Hitmarker_GunplayUpdate_Instigator_Critical]
 modulationSettings {
  volumeRouting=Union
  volumeModulationDestination {bEnableModulation=true, modulators [&Patch_FortHitNotify_Vol]}
 }
}
