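# One-shot stereo MetaSound for the instigator's melee hitmarker on shield, built on the Fort.HitMarker interface.
# Accumulated instigator damage is normalized, shaped by a WaveTable bank curve and remapped to wave player pitch shift; the random pitch inputs are unused.
# A hitmarker volume modulation patch scales both channels.
# Bus sends feed sidechain buses for weapons, explosions, ambience and other categories.
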MetaSoundSource /Game/Athena/Sounds/HitNotifications/Gunplay_Update/Melee/MSS_Hitmarker_Melee_Instigator_Shield
inputs [
 HitMarker.Azimuth {type=Float}
 HitMarker.DamageDealt {type=Float}
 HitMarker.Distance {type=Float}
 HitMarker.InstigatorAccumulatedDamage {type=Float}
 HitMarker.ReceiverPercentHealth {type=Float}
 HitMarker.ReceiverPercentShield {type=Float}
 Random.Pitch.Base {type=Float}
 Random.Pitch.Spread {type=Float}
 Sound {type=WaveAsset}
 UE.Source.OnPlay {type=Trigger}
]
interfaces [Fort.HitMarker:1.0, UE.OutputFormat.Stereo:1.0, UE.Source.OneShot:1.0, UE.Source:1.1]
outputs [
 "UE.OutputFormat.Stereo.Audio:0" {type=Audio}
 "UE.OutputFormat.Stereo.Audio:1" {type=Audio}
 UE.Source.OneShot.OnFinished {type=Trigger}
]
pages [Default {
  input_defaults {
   HitMarker.Azimuth=0
   HitMarker.DamageDealt=0
   HitMarker.Distance=0
   HitMarker.InstigatorAccumulatedDamage=0
   HitMarker.ReceiverPercentHealth=1
   HitMarker.ReceiverPercentShield=1
   Random.Pitch.Base=0.124
   Random.Pitch.Spread=1
   Sound=&/Hitmarker_Melee_Instigator_Shield_01
   UE.Source.OnPlay=null
  }
  nodes [
   n1 = Modulation.GetModulatorValue() {also {defaults {Modulator=/Game/Athena/Sounds/HitNotifications/Gunplay_Update/Modulation/Patch_Hitmarkers_Instigator_Standard_Volume}}}
   n2 = MapRange.MapRange:Float(In=^HitMarker.InstigatorAccumulatedDamage) {also {defaults {In Range B=300}}}
   WavePlayer_0 = UE.Wave Player:Stereo(
    Wave Asset=^Sound
    Play=^UE.Source.OnPlay
    Pitch Shift=MapRange.MapRange:Float(In=UE.WaveTableBankEvaluate(Input=@n2.Out Value) {also {defaults {Index=1, WaveTableBank=/Game/Athena/Sounds/HitNotifications/Gunplay_Update/WB_Hitmarker_GunplayUpdate}}}.Output) {also {defaults {Out Range A=-3, Out Range B=5}}}.Out Value
   ) {also {version [1,1]}}
   Multiply_0 = UE.Multiply:Audio by Float(AdditionalOperands=@n1.Out, PrimaryOperand=@WavePlayer_0.Out Right)
   Multiply_1 = UE.Multiply:Audio by Float(AdditionalOperands=@n1.Out, PrimaryOperand=@WavePlayer_0.Out Left)
  ]
  links [
   {dst="@UE.OutputFormat.Stereo.Audio:1", src=Multiply_0.Out}
   {dst="@UE.OutputFormat.Stereo.Audio:0", src=Multiply_1.Out}
   {dst="@UE.Source.OneShot.OnFinished", src=WavePlayer_0.On Finished}
  ]
 }]
properties {
 assetImportData=&self:AssetImportData
 audioPropertiesSheet=/Game/Athena/Sounds/HitNotifications/Gunplay_Update/APS_Hitmarker_Instigator_Standard
 numChannels=2
 outputFormat=Stereo
 sampleRate=48000
 soundClassObject=&Fort_Hitmarkers
 soundSubmixObject=&Fortnite_SFX
 volume=1.5
 busSends [
  {audioBus=&AB_Sidechain_FortWeapon_Vol, sendLevel=0.35}
  {audioBus=&AB_Sidechain_FortExplosion_Vol, sendLevel=0.35}
  {audioBus=&AB_Sidechain_FortAmbient_Vol, sendLevel=0.3}
  {audioBus=&AB_Sidechain_FortLoot_Vol, sendLevel=0.3}
  {audioBus=&AB_Sidechain_FortVehicles_Vol, sendLevel=0.3}
  {audioBus=&AB_Sidechain_FortCreatures_Vol, sendLevel=0.3}
  {audioBus=&AB_Sidechain_FortEmote_Vol, sendLevel=0.3}
 ]
 concurrencySet [/Game/Athena/Sounds/HitNotifications/Gunplay_Update/CON_Hitmarker_GunplayUpdate_Instigator_Standard]
 modulationSettings {
  volumeRouting=Union
  volumeModulationDestination {bEnableModulation=true, modulators [&Patch_FortHitNotify_Vol]}
 }
}
